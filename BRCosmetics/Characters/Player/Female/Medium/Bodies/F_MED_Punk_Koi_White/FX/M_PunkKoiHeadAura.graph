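# Unlit additive Niagara sprite material for the Punk Koi White character's head aura, glowing in teal-to-green.
# A panning polar-UV mask is subtracted from a radial SphereMask for an eroded shape, and also offsets the twist center of a second polar UV set.
# That set scrolls a spooky-faces texture, combined with the shape to drive brightness and opacity.
# Particle color tints and fades it, with depth fade bypassed on mobile.
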Material /BRCosmetics/Characters/Player/Female/Medium/Bodies/F_MED_Punk_Koi_White/FX/M_PunkKoiHeadAura
graph {
 properties {bUsedWithNiagaraSprites=true, blendMode=BLEND_Additive, pixelDepthOffsetMode=PDOM_Legacy, shadingModel=MSM_Unlit, usedShadingModels=Unlit}
 nodes [
  n1 = Constant3Vector() {constant=<rgba 0 1 0.7647 0>}
  n2 = Constant3Vector() {constant=<rgba 0 0.7098 0.2667 0>}
  n3 = TextureCoordinate()
  n4 = Constant() {r=0.5}
  n5 = Constant()
  n6 = MaterialFunctionCall!(UV=TextureCoordinate(), Twist Center, Twist Edges=Constant() {r=-0.1}) {materialFunction=/Game/Accessories/FORT_Backpacks/Backpack_F_MED_Celestial_Glow/FX/Materials/MF_PolarUV}
  n7 = Constant2Vector() {g=0.25, r=2}
  n8 = TextureSample!(
   UVs=Panner(Coordinate=Multiply(A=@n6.Polar UVs, B=@n7)) {speedX=-0.01, speedY=-0.1}
   Tex
   "DDX(UVs)"=@n6.DDX
   "DDY(UVs)"=@n6.DDX
   Apply View MipBias
  ) {mipValueMode=TMVM_Derivative, samplerType=SAMPLERTYPE_LinearColor, texture=/Game/Effects/Fort_Effects/BluGlow/Violentium/Violentium_Mask}
  n9 = Subtract(A=Multiply(A=SphereMask(A=@n3, B=@n4, Radius=@n4, Hardness=@n5)) {constB=2}, B=@n8.R)
  n10 = Saturate(None=@n9)
  n11 = Saturate(None=Power(Base=@n10) {constExponent=1})
  n12 = MaterialFunctionCall(UV=TextureCoordinate(), Twist Center=Multiply(A=@n8.R) {constB=0.2}) {materialFunction=&MF_PolarUV}
  n13 = Panner(Coordinate=Multiply(A=@n12.Polar UVs, B=Constant2Vector() {g=0.5, r=4})) {speedX=0.05, speedY=-0.2}
  n14 = Multiply(
   A=Power(Base=TextureSample(UVs=@n13, "DDX(UVs)"=@n12.DDX, "DDY(UVs)"=@n12.DDY) {mipValueMode=TMVM_Derivative, texture=/Game/Effects/Fort_Effects/Textures/Tile/T_SpookyFaces_Chroma_Tile}.RGB) {constExponent=0.5}
   B=@n10
  )
  n15 = Add(A=@n14, B=Power(Base=@n9) {constExponent=6})
  n16 = Add(A=Multiply(A=@n14) {constB=2}, B=@n15)
  n17 = ParticleColor()
  n18 = Multiply(A=Multiply(A=LinearInterpolate(A=@n1, B=@n2, Alpha=@n11), B=Saturate(None=@n16)), B=@n17.RGB)
  MaterialFunctionCall(Emissive=@n18) {materialFunction=&MF_FX_EmissiveAdjustments}
  n19 = DepthFade() {fadeDistanceDefault=200}
  Multiply(
   A=ShadingPathSwitch(Default=@n19, Deferred=@n19, Mobile=Constant() {r=1}) {desc=Migrated from FeatureLevelSwitch}
   B=Multiply(A=@n17.A, B=@n15)
  )
  Reroute(None=@n18)
 ]
}
