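# Physics asset for the sleeves of the female medium Punk Koi White cosmetic body, apparently for secondary sleeve motion.
# Each dyn_sleeve bone has a sphere body constrained to its lowerarm capsule, with tight twist limits and a wider swing limit on the right than the left.
# Both lowerarm bodies, which are tiny capsules, link to a pelvis capsule through rotationally free constraints.
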PhysicsAsset /BRCosmetics/Characters/Player/Female/Medium/Bodies/F_MED_Punk_Koi_White/Meshes/F_MED_Punk_Koi_White_sleeve_Physics
bodies {
 dyn_sleeve_l [{shapeType=Sphere, radius=7, shapeName=dyn_sleeve_l_sphere, center=<10 0 0>}]
 dyn_sleeve_r [{shapeType=Sphere, radius=7, shapeName=dyn_sleeve_r_sphere, center=<-10 0 0>}]
 lowerarm_l [{shapeType=Capsule, length=0.505, radius=0.505, shapeName=lowerarm_l_capsule, rotation=<rot 0 -90 0>}]
 lowerarm_r [{shapeType=Capsule, length=0.505, radius=0.505, shapeName=lowerarm_r_capsule, rotation=<rot 0 -90 0>}]
 pelvis [{shapeType=Capsule, length=7.458, radius=8.0636, center=<1.1549 -1.274 0>}]
}
constraints [
 {
  bone1Name=lowerarm_l
  bone2Name=pelvis
  swing1LimitDegrees=45
  swing1Motion=Free
  swing2LimitDegrees=45
  swing2Motion=Free
  twistLimitDegrees=45
  twistMotion=Free
 }
 {
  bone1Name=lowerarm_r
  bone2Name=pelvis
  swing1LimitDegrees=45
  swing1Motion=Free
  swing2LimitDegrees=45
  swing2Motion=Free
  twistLimitDegrees=45
  twistMotion=Free
 }
 {bone1Name=dyn_sleeve_r, bone2Name=lowerarm_r, swing1LimitDegrees=45, swing2LimitDegrees=45, twistLimitDegrees=2}
 {bone1Name=dyn_sleeve_l, bone2Name=lowerarm_l, swing1LimitDegrees=30, swing2LimitDegrees=30, twistLimitDegrees=2}
]
