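# Physics asset for the Punk Koi White female face accessory (NoBack variant), giving capsule bodies to dynamic hair and ponytail bones.
# Upper spine, clavicle, neck and head capsules form the body volume, with hair root bodies linked to the head by free-rotation constraints.
# Limited swing and twist joints drive three front-to-tip hair chains, cross-linked side to center, plus paired hair2 chains.
# Each ponytail fans three bodies off its second segment.
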PhysicsAsset /BRCosmetics/Characters/Player/Female/Medium/Bodies/F_MED_Punk_Koi_White/Meshes/Parts/F_MED_Punk_Koi_White_FaceAcc_Physics_NoBack
bodies {
 clavicle_l [{shapeType=Capsule, length=12.25, radius=6.5, center=<9.58 0.6782 -3.4157>, rotation=<rot 74.5066 0 0>}]
 clavicle_r [{shapeType=Capsule, length=13, radius=6.5, center=<-9.0601 -0.6784 3.4691>, rotation=<rot 74.6207 -0.0001 -180>}]
 dyn_hair_a [{shapeType=Capsule, length=5.5, radius=1.75, center=<5 0 0>, rotation=<rot 90 0 0>}]
 dyn_hair_a_base [{shapeType=Capsule, length=1, radius=1}]
 dyn_hair_a_base_l [{shapeType=Capsule, length=1, radius=1}]
 dyn_hair_a_base_r [{shapeType=Capsule, length=1, radius=1}]
 dyn_hair_a_l [{shapeType=Capsule, length=7.75, radius=2.25, center=<5 0 0>, rotation=<rot 90 0 0>}]
 dyn_hair_a_r [{shapeType=Capsule, length=5, radius=2.5, center=<-5 0 0>, rotation=<rot 90 0 180>}]
 dyn_hair_b [{shapeType=Capsule, length=7, radius=2, center=<5 0 0>, rotation=<rot 90 0 0>}]
 dyn_hair_b_l [{shapeType=Capsule, length=9, radius=1.5, center=<5 0 0>, rotation=<rot 90 0 0>}]
 dyn_hair_b_r [{shapeType=Capsule, length=10.75, radius=1.5, center=<-4 0 0>, rotation=<rot 90 0 -179.9999>}]
 dyn_hair_c [{shapeType=Capsule, length=9, radius=1.75, center=<5 0 0>, rotation=<rot 90 0 0>}]
 dyn_hair_c_l [{shapeType=Capsule, length=7.25, radius=1, center=<4 0.0001 0>, rotation=<rot 90 0 0>}]
 dyn_hair_c_r [{shapeType=Capsule, length=6, radius=1.5, center=<-4 0 0>, rotation=<rot 90 0 180>}]
 dyn_ponytail_l_01 [{shapeType=Capsule, length=0.505, radius=0.505, shapeName=dyn_ponytail_l_01_capsule, rotation=<rot 0 -90 0>}]
 dyn_ponytail_l_02 [{shapeType=Capsule, length=9.8077, radius=3.692, shapeName=dyn_ponytail_l_02_capsule, center=<11.1527 0.0002 0>, rotation=<rot 90 0 0>}]
 dyn_ponytail_l_03 [{shapeType=Capsule, length=2.13, radius=1.63, shapeName=dyn_ponytail_l_03_capsule, center=<2.6467 0 0>, rotation=<rot 90 0 0>}]
 dyn_ponytail_l_04 [{shapeType=Capsule, length=2.13, radius=1.63, shapeName=dyn_ponytail_l_04_capsule, center=<2.9108 0 0>, rotation=<rot 90 0 0>}]
 dyn_ponytail_l_05 [{shapeType=Capsule, length=3.1963, radius=1.9808, shapeName=dyn_ponytail_l_05_capsule, center=<3.7604 -0.0001 0>, rotation=<rot 90 0 0>}]
 dyn_ponytail_r_01 [{shapeType=Capsule, length=0.505, radius=0.505, shapeName=dyn_ponytail_r_01_capsule, rotation=<rot 0 -90 0>}]
 dyn_ponytail_r_02 [{shapeType=Capsule, length=11.0277, radius=3.312, shapeName=dyn_ponytail_r_02_capsule, center=<11.6527 0.0002 0>, rotation=<rot 90 0 0>}]
 dyn_ponytail_r_03 [{shapeType=Capsule, length=4.174, radius=1.6749, shapeName=dyn_ponytail_r_03_capsule, center=<3.3636 -0.0001 0>, rotation=<rot 90 0 0>}]
 dyn_ponytail_r_04 [{shapeType=Capsule, length=2.13, radius=1.63, shapeName=dyn_ponytail_r_04_capsule, center=<2.8858 0 0>, rotation=<rot 90 0 0>}]
 dyn_ponytail_r_05 [{shapeType=Capsule, length=2.9893, radius=1.7716, shapeName=dyn_ponytail_r_05_capsule, center=<3.3031 -0.0001 0>, rotation=<rot 90 0 0>}]
 hair2_0_l [{shapeType=Capsule, length=0.505, radius=0.505, shapeName=hair2_0_l_capsule, rotation=<rot 0 -90 0>}]
 hair2_0_r [{shapeType=Capsule, length=0.505, radius=0.505, shapeName=hair2_0_r_capsule, rotation=<rot 0 -90 0>}]
 hair2_1_l [{shapeType=Capsule, length=8, radius=2, shapeName=hair2_1_l_capsule, center=<5 0 0>, rotation=<rot 90 0 0>}]
 hair2_1_r [{shapeType=Capsule, length=8, radius=2, shapeName=hair2_1_r_capsule, center=<5 0 0>, rotation=<rot 90 0 0>}]
 hair2_2_l [{shapeType=Capsule, length=5.875, radius=2, shapeName=hair2_2_l_capsule, center=<4.5 0 0>, rotation=<rot 90 0 0>}]
 hair2_2_r [{shapeType=Capsule, length=5.875, radius=2, shapeName=hair2_2_r_capsule, center=<4.5 0 0>, rotation=<rot 90 0 0>}]
 head [{shapeType=Capsule, length=8, radius=6, shapeName=head_capsule, center=<0 2.3 0>, rotation=<rot 90 0 0>}]
 neck_02 [{shapeType=Capsule, length=4.75, radius=5.25, center=<-1.9548 0 0>, rotation=<rot 90 0 0>}]
 spine_05 [{shapeType=Capsule, length=36.25, radius=11.75, center=<-2.3162 3.0345 0>}]
}
constraints [
 {
  bone1Name=dyn_hair_a_base
  bone2Name=head
  swing1LimitDegrees=45
  swing1Motion=Free
  swing2LimitDegrees=45
  swing2Motion=Free
  twistLimitDegrees=45
  twistMotion=Free
 }
 {
  bone1Name=dyn_hair_a_base_l
  bone2Name=head
  swing1LimitDegrees=45
  swing1Motion=Free
  swing2LimitDegrees=45
  swing2Motion=Free
  twistLimitDegrees=45
  twistMotion=Free
 }
 {
  bone1Name=dyn_hair_a_base_r
  bone2Name=head
  swing1LimitDegrees=45
  swing1Motion=Free
  swing2LimitDegrees=45
  swing2Motion=Free
  twistLimitDegrees=45
  twistMotion=Free
 }
 {bone1Name=dyn_hair_a, bone2Name=dyn_hair_a_base, swing1LimitDegrees=20, swing2LimitDegrees=10, twistLimitDegrees=5}
 {bone1Name=dyn_hair_b, bone2Name=dyn_hair_a, swing1LimitDegrees=20, swing2LimitDegrees=10, twistLimitDegrees=5}
 {bone1Name=dyn_hair_c, bone2Name=dyn_hair_b, swing1LimitDegrees=20, swing2LimitDegrees=10, twistLimitDegrees=5}
 {
  bone1Name=neck_02
  bone2Name=spine_05
  swing1LimitDegrees=45
  swing1Motion=Free
  swing2LimitDegrees=45
  swing2Motion=Free
  twistLimitDegrees=45
  twistMotion=Free
 }
 {bone1Name=dyn_hair_c_l, bone2Name=dyn_hair_b_l, swing1LimitDegrees=5, swing2LimitDegrees=10, twistLimitDegrees=5}
 {bone1Name=dyn_hair_a_l, bone2Name=dyn_hair_a_base_l, swing1LimitDegrees=5, swing2LimitDegrees=10, twistLimitDegrees=5}
 {bone1Name=dyn_hair_b_l, bone2Name=dyn_hair_a_l, swing1LimitDegrees=5, swing2LimitDegrees=10, twistLimitDegrees=5}
 {bone1Name=dyn_hair_c_r, bone2Name=dyn_hair_b_r, swing1LimitDegrees=5, swing2LimitDegrees=10, twistLimitDegrees=5}
 {bone1Name=dyn_hair_b_r, bone2Name=dyn_hair_a_r, swing1LimitDegrees=5, swing2LimitDegrees=10, twistLimitDegrees=5}
 {bone1Name=dyn_hair_a_r, bone2Name=dyn_hair_a_base_r, swing1LimitDegrees=5, swing2LimitDegrees=10, twistLimitDegrees=5}
 {
  bone1Name=clavicle_l
  bone2Name=spine_05
  swing1LimitDegrees=45
  swing1Motion=Free
  swing2LimitDegrees=45
  swing2Motion=Free
  twistLimitDegrees=45
  twistMotion=Free
 }
 {
  bone1Name=clavicle_r
  bone2Name=spine_05
  swing1LimitDegrees=45
  swing1Motion=Free
  swing2LimitDegrees=45
  swing2Motion=Free
  twistLimitDegrees=45
  twistMotion=Free
 }
 {bone1Name=dyn_hair_c_r, bone2Name=dyn_hair_c, swing1LimitDegrees=10, swing2LimitDegrees=20, twistLimitDegrees=5}
 {bone1Name=dyn_hair_c_l, bone2Name=dyn_hair_c, swing1LimitDegrees=10, swing2LimitDegrees=20, twistLimitDegrees=5}
 {bone1Name=dyn_hair_b_r, bone2Name=dyn_hair_b, swing1LimitDegrees=10, swing2LimitDegrees=20, twistLimitDegrees=5}
 {bone1Name=dyn_hair_b_l, bone2Name=dyn_hair_b, swing1LimitDegrees=10, swing2LimitDegrees=20, twistLimitDegrees=5}
 {
  bone1Name=hair2_0_l
  bone2Name=head
  swing1LimitDegrees=45
  swing1Motion=Free
  swing2LimitDegrees=45
  swing2Motion=Free
  twistLimitDegrees=45
  twistMotion=Free
 }
 {bone1Name=hair2_1_l, bone2Name=hair2_0_l, swing1LimitDegrees=20, swing2LimitDegrees=30, twistLimitDegrees=2}
 {bone1Name=hair2_2_l, bone2Name=hair2_1_l, swing1LimitDegrees=20, swing2LimitDegrees=30, twistLimitDegrees=2}
 {
  bone1Name=hair2_0_r
  bone2Name=head
  swing1LimitDegrees=45
  swing1Motion=Free
  swing2LimitDegrees=45
  swing2Motion=Free
  twistLimitDegrees=45
  twistMotion=Free
 }
 {bone1Name=hair2_1_r, bone2Name=hair2_0_r, swing1LimitDegrees=20, swing2LimitDegrees=30, twistLimitDegrees=2}
 {bone1Name=hair2_2_r, bone2Name=hair2_1_r, swing1LimitDegrees=20, swing2LimitDegrees=30, twistLimitDegrees=2}
 {
  bone1Name=dyn_ponytail_l_01
  bone2Name=head
  swing1LimitDegrees=45
  swing1Motion=Free
  swing2LimitDegrees=45
  swing2Motion=Free
  twistLimitDegrees=45
  twistMotion=Free
 }
 {bone1Name=dyn_ponytail_l_02, bone2Name=dyn_ponytail_l_01, swing1LimitDegrees=75, swing2LimitDegrees=75, twistLimitDegrees=5}
 {bone1Name=dyn_ponytail_l_03, bone2Name=dyn_ponytail_l_02, swing1LimitDegrees=30, swing2LimitDegrees=30, twistLimitDegrees=5}
 {bone1Name=dyn_ponytail_l_04, bone2Name=dyn_ponytail_l_02, swing1LimitDegrees=30, swing2LimitDegrees=30, twistLimitDegrees=5}
 {bone1Name=dyn_ponytail_l_05, bone2Name=dyn_ponytail_l_02, swing1LimitDegrees=30, swing2LimitDegrees=30, twistLimitDegrees=5}
 {
  bone1Name=dyn_ponytail_r_01
  bone2Name=head
  swing1LimitDegrees=45
  swing1Motion=Free
  swing2LimitDegrees=45
  swing2Motion=Free
  twistLimitDegrees=45
  twistMotion=Free
 }
 {bone1Name=dyn_ponytail_r_02, bone2Name=dyn_ponytail_r_01, swing1LimitDegrees=75, swing2LimitDegrees=75, twistLimitDegrees=5}
 {bone1Name=dyn_ponytail_r_03, bone2Name=dyn_ponytail_r_02, swing1LimitDegrees=30, swing2LimitDegrees=30, twistLimitDegrees=5}
 {bone1Name=dyn_ponytail_r_04, bone2Name=dyn_ponytail_r_02, swing1LimitDegrees=30, swing2LimitDegrees=30, twistLimitDegrees=5}
 {bone1Name=dyn_ponytail_r_05, bone2Name=dyn_ponytail_r_02, swing1LimitDegrees=30, swing2LimitDegrees=30, twistLimitDegrees=5}
 {
  bone1Name=head
  bone2Name=neck_02
  swing1LimitDegrees=45
  swing1Motion=Free
  swing2LimitDegrees=45
  swing2Motion=Free
  twistLimitDegrees=45
  twistMotion=Free
 }
]
